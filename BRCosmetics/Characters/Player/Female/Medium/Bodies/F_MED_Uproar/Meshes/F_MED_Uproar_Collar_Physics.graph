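# Physics asset for the collar of the F_MED_Uproar female player cosmetic body, covering the Dyn_Collar bones and nearby upper-body bones.
# Front, side and back collar capsules are constrained to small Dyn_Master anchor bodies with tight angular limits, and the back forms a two-link chain.
# Capsules on spine_05, neck, clavicles and upper arms form the body volume around the collar, joined by free-rotating constraints.
# No constraint links the master anchors to those body capsules.
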PhysicsAsset /BRCosmetics/Characters/Player/Female/Medium/Bodies/F_MED_Uproar/Meshes/F_MED_Uproar_Collar_Physics
bodies {
 Dyn_Collar_B1 [{shapeType=Capsule, length=10, radius=5, shapeName=Dyn_Collar_B1_capsule, center=<2.4 -4.4921 -0.0001>}]
 Dyn_Collar_B3 [{shapeType=Capsule, length=6, radius=2, shapeName=Dyn_Collar_B3_capsule, center=<1 -0.25 0>}]
 Dyn_Collar_FL1 [
  {shapeType=Capsule, length=3, radius=5, shapeName=Dyn_Collar_FL1_capsule, center=<5.6001 -4.7 0>, rotation=<rot 90 0 0>}
  {shapeType=Capsule, length=3, radius=5, shapeName=Dyn_Collar_FL1_capsule, center=<6.7088 -4.045 -4.537>, rotation=<rot 90 0 0>}
 ]
 Dyn_Collar_L2 [{shapeType=Capsule, length=4, radius=4, shapeName=Dyn_Collar_L2_capsule, center=<0.1524 1.0103 -2.5635>, rotation=<rot 90 90 50>}]
 Dyn_Collar_R2 [{shapeType=Capsule, length=4, radius=4, shapeName=Dyn_Collar_R2_capsule, center=<0.1523 1.0103 2.3848>, rotation=<rot 90 -14.0363 -54.0362>}]
 Dyn_Collar__FR1 [
  {shapeType=Capsule, length=3, radius=5, shapeName=Dyn_Collar__FR1_capsule, center=<5.6 -4.6 0>, rotation=<rot 90 0 0>}
  {shapeType=Capsule, length=3, radius=5, shapeName=Dyn_Collar__FR1_capsule, center=<5.6 -4.6 4.0657>, rotation=<rot 90 0 0>}
 ]
 Dyn_Master_Collar_B [{shapeType=Capsule, length=0.5, radius=0.5, shapeName=Dyn_Master_Collar_B_capsule}]
 Dyn_Master_Collar_FL [{shapeType=Capsule, length=0.5, radius=0.5, shapeName=Dyn_Master_Collar_FL_capsule}]
 Dyn_Master_Collar_L [{shapeType=Capsule, length=0.5, radius=0.5, shapeName=Dyn_Master_Collar_L_capsule}]
 Dyn_Master_Collar_R [{shapeType=Capsule, length=0.5, radius=0.5, shapeName=Dyn_Master_Collar_R_capsule}]
 Dyn_Master_Collar__FR [{shapeType=Capsule, length=0.5, radius=0.5, shapeName=Dyn_Master_Collar__FR_capsule}]
 clavicle_l [{shapeType=Capsule, length=3.5, radius=7, shapeName=clavicle_l_capsule, center=<7.5094 -1.7219 -0.8269>, rotation=<rot 90 0 0>}]
 clavicle_r [{shapeType=Capsule, length=2, radius=8, shapeName=clavicle_r_capsule, center=<-8.2806 1.3021 1.3978>, rotation=<rot 90 0 0>}]
 neck_01 [{shapeType=Capsule, length=3, radius=4.8, shapeName=neck_01_capsule, center=<3.2493 0.0001 0>, rotation=<rot 90 0 0>}]
 spine_05 [{shapeType=Capsule, length=11.814, radius=8, shapeName=spine_05_capsule, center=<3.008 -1.7536 0>}]
 upperarm_l [{shapeType=Capsule, length=10, radius=6, shapeName=upperarm_l_capsule, center=<9.0837 -0.6692 -0.0001>, rotation=<rot 90 0 0>}]
 upperarm_r [{shapeType=Capsule, length=10, radius=6, shapeName=upperarm_r_capsule, center=<-8.7143 0.5885 -0.0001>, rotation=<rot 90 0 0>}]
}
constraints [
 {
  bone1Name=upperarm_l
  bone2Name=clavicle_l
  swing1LimitDegrees=45
  swing1Motion=Free
  swing2LimitDegrees=45
  swing2Motion=Free
  twistLimitDegrees=45
  twistMotion=Free
 }
 {
  bone1Name=upperarm_r
  bone2Name=clavicle_r
  swing1LimitDegrees=45
  swing1Motion=Free
  swing2LimitDegrees=45
  swing2Motion=Free
  twistLimitDegrees=45
  twistMotion=Free
 }
 {bone1Name=Dyn_Collar_FL1, bone2Name=Dyn_Master_Collar_FL, swing1LimitDegrees=30, swing2LimitDegrees=5, twistLimitDegrees=1}
 {bone1Name=Dyn_Collar__FR1, bone2Name=Dyn_Master_Collar__FR, swing1LimitDegrees=30, swing2LimitDegrees=5, twistLimitDegrees=1}
 {bone1Name=Dyn_Collar_B1, bone2Name=Dyn_Master_Collar_B, swing1LimitDegrees=30, swing2LimitDegrees=30, twistLimitDegrees=30}
 {
  bone1Name=clavicle_l
  bone2Name=spine_05
  swing1LimitDegrees=45
  swing1Motion=Free
  swing2LimitDegrees=45
  swing2Motion=Free
  twistLimitDegrees=45
  twistMotion=Free
 }
 {
  bone1Name=clavicle_r
  bone2Name=spine_05
  swing1LimitDegrees=45
  swing1Motion=Free
  swing2LimitDegrees=45
  swing2Motion=Free
  twistLimitDegrees=45
  twistMotion=Free
 }
 {
  bone1Name=neck_01
  bone2Name=spine_05
  swing1LimitDegrees=45
  swing1Motion=Free
  swing2LimitDegrees=45
  swing2Motion=Free
  twistLimitDegrees=45
  twistMotion=Free
 }
 {bone1Name=Dyn_Collar_B3, bone2Name=Dyn_Collar_B1, swing1LimitDegrees=20, swing2LimitDegrees=1, twistLimitDegrees=1}
 {bone1Name=Dyn_Collar_L2, bone2Name=Dyn_Master_Collar_L, swing1LimitDegrees=20, swing2LimitDegrees=1, twistLimitDegrees=1}
 {bone1Name=Dyn_Collar_R2, bone2Name=Dyn_Master_Collar_R, swing1LimitDegrees=20, swing2LimitDegrees=1, twistLimitDegrees=1}
]
