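# Physics asset for the F_MED_Uproar female body cosmetic that sets up the jacket's dynamic bones for physics motion.
# Capsule chains around the front, sides and back hang from spine_02, with three links each except the back chains, which have two.
# Jacket joints share tight limits that allow swing mostly on one axis, with almost no twist.
# Pelvis, spine and thigh capsules form the body, with locked spine joints and free thigh joints.
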PhysicsAsset /BRCosmetics/Characters/Player/Female/Medium/Bodies/F_MED_Uproar/Meshes/F_MED_Uproar_Physics
bodies {
 Dyn_Jacket_B2 [{shapeType=Capsule, length=2, radius=6, shapeName=Dyn_Jacket_B2_capsule, center=<5 -6.3 0>, rotation=<rot 90 0 0>}]
 Dyn_Jacket_B3 [{shapeType=Capsule, length=1, radius=6, shapeName=Dyn_Jacket_B3_capsule, center=<5 -7.5 0>, rotation=<rot 90 0 0>}]
 Dyn_Jacket_BL2 [{shapeType=Capsule, length=2, radius=6, shapeName=Dyn_Jacket_BL2_capsule, center=<5 -6.3 0>, rotation=<rot 90 0 0>}]
 Dyn_Jacket_BL3 [{shapeType=Capsule, length=1, radius=6, shapeName=Dyn_Jacket_BL3_capsule, center=<4.9996 -6.8356 0>, rotation=<rot 90 0 0>}]
 Dyn_Jacket_BR2 [{shapeType=Capsule, length=2, radius=6, shapeName=Dyn_Jacket_BR2_capsule, center=<5 -6.3 0>, rotation=<rot 90 0 0>}]
 Dyn_Jacket_BR3 [{shapeType=Capsule, length=1, radius=6, shapeName=Dyn_Jacket_BR3_capsule, center=<4.9996 -6.8356 0>, rotation=<rot 90 0 0>}]
 Dyn_Jacket_FL1 [{shapeType=Capsule, length=2, radius=6, shapeName=Dyn_Jacket_FL1_capsule, center=<5 -4.6 0>, rotation=<rot 90 0 0>}]
 Dyn_Jacket_FL2 [{shapeType=Capsule, length=4, radius=6, shapeName=Dyn_Jacket_FL2_capsule, center=<5 -6.3 0>, rotation=<rot 90 0 0>}]
 Dyn_Jacket_FL3 [{shapeType=Capsule, length=1, radius=6, shapeName=Dyn_Jacket_FL3_capsule, center=<5 -4.7795 0>, rotation=<rot 90 0 0>}]
 Dyn_Jacket_FM_L1 [{shapeType=Capsule, length=2, radius=6, shapeName=Dyn_Jacket_FM_L1_capsule, center=<5 -4.6 0>, rotation=<rot 90 0 0>}]
 Dyn_Jacket_FM_L2 [{shapeType=Capsule, length=2, radius=6, shapeName=Dyn_Jacket_FM_L2_capsule, center=<5 -6.3 0>, rotation=<rot 90 0 0>}]
 Dyn_Jacket_FM_L3 [{shapeType=Capsule, length=1, radius=6, shapeName=Dyn_Jacket_FM_L3_capsule, center=<4.9996 -6.8 0>, rotation=<rot 90 0 0>}]
 Dyn_Jacket_FM_R1 [{shapeType=Capsule, length=2, radius=6, shapeName=Dyn_Jacket_FM_R1_capsule, center=<5 -4.6 0>, rotation=<rot 90 0 0>}]
 Dyn_Jacket_FM_R2 [{shapeType=Capsule, length=2, radius=6, shapeName=Dyn_Jacket_FM_R2_capsule, center=<5 -6.3 0>, rotation=<rot 90 0 0>}]
 Dyn_Jacket_FM_R3 [{shapeType=Capsule, length=1, radius=6, shapeName=Dyn_Jacket_FM_R3_capsule, center=<5 -6.8 0>, rotation=<rot 90 0 0>}]
 Dyn_Jacket_FR1 [{shapeType=Capsule, length=2, radius=6, shapeName=Dyn_Jacket_FR1_capsule, center=<5 -4.6 0>, rotation=<rot 90 0 0>}]
 Dyn_Jacket_FR2 [{shapeType=Capsule, length=4, radius=6, shapeName=Dyn_Jacket_FR2_capsule, center=<5 -6.3 0>, rotation=<rot 90 0 0>}]
 Dyn_Jacket_FR3 [{shapeType=Capsule, length=1, radius=6, shapeName=Dyn_Jacket_FR3_capsule, center=<5 -4.7795 0>, rotation=<rot 90 0 0>}]
 Dyn_Jacket_L1 [{shapeType=Capsule, length=2, radius=6, shapeName=Dyn_Jacket_L1_capsule, center=<5 -4.6 0>, rotation=<rot 90 0 0>}]
 Dyn_Jacket_L2 [{shapeType=Capsule, length=2, radius=6, shapeName=Dyn_Jacket_L2_capsule, center=<5 -6.3 0>, rotation=<rot 90 0 0>}]
 Dyn_Jacket_L3 [{shapeType=Capsule, length=2, radius=6, shapeName=Dyn_Jacket_L3_capsule, center=<5 0.0002 0>, rotation=<rot 90 0 0>}]
 Dyn_Jacket_R1 [{shapeType=Capsule, length=2, radius=6, shapeName=Dyn_Jacket_R1_capsule, center=<5 -4.6 0>, rotation=<rot 90 0 0>}]
 Dyn_Jacket_R2 [{shapeType=Capsule, length=2, radius=6, shapeName=Dyn_Jacket_R2_capsule, center=<5 -6.3 0>, rotation=<rot 90 0 0>}]
 Dyn_Jacket_R3 [{shapeType=Capsule, length=1, radius=6, shapeName=Dyn_Jacket_R3_capsule, center=<5 -6.5 0>, rotation=<rot 90 0 0>}]
 pelvis [{shapeType=Capsule, length=6.8868, radius=11.1745, shapeName=pelvis_capsule, center=<0 -1.0961 0>}]
 spine_01 [{shapeType=Capsule, length=3, radius=11, shapeName=spine_01_capsule, center=<3.2299 1.2771 0>, rotation=<rot 0 12.9079 7.9077>}]
 spine_02 [{shapeType=Capsule, length=6, radius=9.2, shapeName=spine_02_capsule, center=<0.0001 1.1165 0>, rotation=<rot 90 0 0>}]
 thigh_l [{shapeType=Capsule, length=20, radius=9, shapeName=thigh_l_capsule, center=<-17.0739 -0.0001 0>, rotation=<rot 90 0 0>}]
 thigh_r [{shapeType=Capsule, length=20, radius=9, shapeName=thigh_r_capsule, center=<16.9266 0 0>, rotation=<rot 90 0 0>}]
}
constraints [
 {bone1Name=Dyn_Jacket_B3, bone2Name=Dyn_Jacket_B2, swing1LimitDegrees=20, swing2LimitDegrees=1, twistLimitDegrees=1}
 {bone1Name=Dyn_Jacket_BR3, bone2Name=Dyn_Jacket_BR2, swing1LimitDegrees=20, swing2LimitDegrees=1, twistLimitDegrees=1}
 {bone1Name=Dyn_Jacket_BL3, bone2Name=Dyn_Jacket_BL2, swing1LimitDegrees=20, swing2LimitDegrees=1, twistLimitDegrees=1}
 {bone1Name=Dyn_Jacket_R3, bone2Name=Dyn_Jacket_R2, swing1LimitDegrees=20, swing2LimitDegrees=1, twistLimitDegrees=1}
 {bone1Name=Dyn_Jacket_FM_R3, bone2Name=Dyn_Jacket_FM_R2, swing1LimitDegrees=20, swing2LimitDegrees=1, twistLimitDegrees=1}
 {bone1Name=Dyn_Jacket_FM_L3, bone2Name=Dyn_Jacket_FM_L2, swing1LimitDegrees=20, swing2LimitDegrees=1, twistLimitDegrees=1}
 {bone1Name=Dyn_Jacket_FR3, bone2Name=Dyn_Jacket_FR2, swing1LimitDegrees=20, swing2LimitDegrees=1, twistLimitDegrees=1}
 {bone1Name=Dyn_Jacket_FL3, bone2Name=Dyn_Jacket_FL2, swing1LimitDegrees=20, swing2LimitDegrees=1, twistLimitDegrees=1}
 {bone1Name=Dyn_Jacket_B2, bone2Name=spine_02, swing1LimitDegrees=20, swing2LimitDegrees=1, twistLimitDegrees=1}
 {bone1Name=Dyn_Jacket_BR2, bone2Name=spine_02, swing1LimitDegrees=20, swing2LimitDegrees=1, twistLimitDegrees=1}
 {bone1Name=Dyn_Jacket_BL2, bone2Name=spine_02, swing1LimitDegrees=20, swing2LimitDegrees=1, twistLimitDegrees=1}
 {bone1Name=Dyn_Jacket_R2, bone2Name=Dyn_Jacket_R1, swing1LimitDegrees=20, swing2LimitDegrees=1, twistLimitDegrees=1}
 {bone1Name=Dyn_Jacket_L3, bone2Name=Dyn_Jacket_L2, swing1LimitDegrees=20, swing2LimitDegrees=1, twistLimitDegrees=1}
 {bone1Name=Dyn_Jacket_FM_R2, bone2Name=Dyn_Jacket_FM_R1, swing1LimitDegrees=20, swing2LimitDegrees=1, twistLimitDegrees=1}
 {bone1Name=Dyn_Jacket_FM_L2, bone2Name=Dyn_Jacket_FM_L1, swing1LimitDegrees=20, swing2LimitDegrees=1, twistLimitDegrees=1}
 {bone1Name=Dyn_Jacket_FR2, bone2Name=Dyn_Jacket_FR1, swing1LimitDegrees=20, swing2LimitDegrees=1, twistLimitDegrees=1}
 {bone1Name=Dyn_Jacket_FL2, bone2Name=Dyn_Jacket_FL1, swing1LimitDegrees=20, swing2LimitDegrees=1, twistLimitDegrees=1}
 {bone1Name=Dyn_Jacket_L2, bone2Name=Dyn_Jacket_L1, swing1LimitDegrees=20, swing2LimitDegrees=1, twistLimitDegrees=1}
 {bone1Name=Dyn_Jacket_FL1, bone2Name=spine_02, swing1LimitDegrees=20, swing2LimitDegrees=1, twistLimitDegrees=1}
 {bone1Name=Dyn_Jacket_FR1, bone2Name=spine_02, swing1LimitDegrees=20, swing2LimitDegrees=1, twistLimitDegrees=1}
 {bone1Name=Dyn_Jacket_FM_L1, bone2Name=spine_02, swing1LimitDegrees=20, swing2LimitDegrees=1, twistLimitDegrees=1}
 {bone1Name=Dyn_Jacket_FM_R1, bone2Name=spine_02, swing1LimitDegrees=20, swing2LimitDegrees=1, twistLimitDegrees=1}
 {bone1Name=Dyn_Jacket_L1, bone2Name=spine_02, swing1LimitDegrees=20, swing2LimitDegrees=1, twistLimitDegrees=1}
 {bone1Name=Dyn_Jacket_R1, bone2Name=spine_02, swing1LimitDegrees=20, swing2LimitDegrees=1, twistLimitDegrees=1}
 {
  bone1Name=spine_01
  bone2Name=pelvis
  swing1LimitDegrees=45
  swing1Motion=Locked
  swing2LimitDegrees=45
  swing2Motion=Locked
  twistLimitDegrees=45
  twistMotion=Locked
 }
 {
  bone1Name=spine_02
  bone2Name=spine_01
  swing1LimitDegrees=45
  swing1Motion=Locked
  swing2LimitDegrees=45
  swing2Motion=Locked
  twistLimitDegrees=45
  twistMotion=Locked
 }
 {
  bone1Name=thigh_l
  bone2Name=pelvis
  swing1LimitDegrees=45
  swing1Motion=Free
  swing2LimitDegrees=45
  swing2Motion=Free
  twistLimitDegrees=45
  twistMotion=Free
 }
 {
  bone1Name=thigh_r
  bone2Name=pelvis
  swing1LimitDegrees=45
  swing1Motion=Free
  swing2LimitDegrees=45
  swing2Motion=Free
  twistLimitDegrees=45
  twistMotion=Free
 }
]
