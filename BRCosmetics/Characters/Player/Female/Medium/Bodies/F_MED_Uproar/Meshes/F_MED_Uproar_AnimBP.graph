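# Animation Blueprint for the F_MED_Uproar cosmetic body, adding deformation and secondary motion to clothing on the attached parent's pose.
# Copied pose runs through sculpter and twist Control Rigs, with a collar adjustment chosen by bIsInFrontEnd, then two RigidBody simulations.
# Additive bone offsets then correct belt, jacket and collar, driven by crouch state or arm, hood and leg alphas computed each update from bone transform deltas.
# Pendent and hoodie offset nodes are unconnected.
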AnimBlueprint /BRCosmetics/Characters/Player/Female/Medium/Bodies/F_MED_Uproar/Meshes/F_MED_Uproar_AnimBP
target_skeleton=&/F_MED_Uproar_Skeleton
graphs [
 AnimGraph {nodes [
   n1 = Get Right Arm ADS Up() -> (Right Arm ADS Up:double)
   "Transform (Modify) Bone - Bone: Dyn_Master_Pendent"(Alpha=@n1, Translation="0,1.5,0", Rotation="-15,0,0", Scale="1,1,1") {
    rotationMode=BMM_Additive
    rotationSpace=BCS_BoneSpace
    scaleMode=BMM_Additive
    scaleSpace=BCS_BoneSpace
    translationMode=BMM_Additive
    translationSpace=BCS_BoneSpace
    alphaBoolBlend {blendOption=Linear}
    boneToModify {boneName=Dyn_Master_Pendent}
    rotation {pitch=-15}
    translation {y=1.5}
   }
   n2 = "Get Left  Arm ADS Up"() -> ("Left  Arm ADS Up":double)
   "Transform (Modify) Bone - Bone: Dyn_Master_Hoodie"(Alpha=@n2, Translation="2,0,0", Rotation="0,0,0", Scale="1,1,1") {
    rotationMode=BMM_Additive
    rotationSpace=BCS_BoneSpace
    scaleMode=BMM_Additive
    scaleSpace=BCS_BoneSpace
    translationMode=BMM_Additive
    translationSpace=BCS_BoneSpace
    alphaBoolBlend {blendOption=Linear}
    boneToModify {boneName=Dyn_Master_Hoodie}
    translation {x=2}
   }
   n3 = Control Rig(Source=Copy Pose From Mesh() {bUseAttachedParent=true}) {
    alphaBoolBlend {blendOption=Linear}
    controlRigAssetReference {blueprintRigClass=&/F_MED_Uproar_Deform_CtrlRig_sculpters.F_MED_Uproar_Deform_CtrlRig_sculpters_C}
   }
   n4 = Control Rig(Source=@n3) {
    alphaBoolBlend {blendOption=Linear}
    controlRigAssetReference {blueprintRigClass=&/F_MED_Uproar_Deform_CtrlRig_twists.F_MED_Uproar_Deform_CtrlRig_twists_C}
   }
   n5 = "Transform (Modify) Bone - Bone: Dyn_Master_Collar_L"(ComponentPose=Local To Component(LocalPose=@n4), Alpha=1, Translation="0,0,0", Rotation="25,-2,0", Scale="1,1,1") {
    rotationMode=BMM_Additive
    rotationSpace=BCS_BoneSpace
    scaleMode=BMM_Additive
    scaleSpace=BCS_BoneSpace
    translationMode=BMM_Additive
    translationSpace=BCS_BoneSpace
    alphaBoolBlend {blendOption=Linear}
    boneToModify {boneName=Dyn_Master_Collar_L}
    rotation {pitch=25, yaw=-2}
   }
   n6 = Component To Local(ComponentPose=@n5)
   n7 = Control Rig(Source=Copy Pose From Mesh() {bUseAttachedParent=true}) {
    alphaBoolBlend {blendOption=Linear}
    controlRigAssetReference {blueprintRigClass=&F_MED_Uproar_Deform_CtrlRig_sculpters_C}
   }
   n8 = Control Rig(Source=@n7) {
    alphaBoolBlend {blendOption=Linear}
    controlRigAssetReference {blueprintRigClass=&F_MED_Uproar_Deform_CtrlRig_twists_C}
   }
   n9 = "Transform (Modify) Bone - Bone: Dyn_Master_Collar_L"(
    ComponentPose=Local To Component(LocalPose=@n8)
    Alpha=Get Left_Hood() -> (Left_Hood:double)
    Translation="0,0,0"
    Rotation="0,30,0"
    Scale="1,1,1"
   ) {
    rotationMode=BMM_Additive
    rotationSpace=BCS_BoneSpace
    scaleMode=BMM_Additive
    scaleSpace=BCS_BoneSpace
    translationMode=BMM_Additive
    translationSpace=BCS_BoneSpace
    alphaBoolBlend {blendOption=Linear}
    boneToModify {boneName=Dyn_Master_Collar_L}
    rotation {yaw=30}
   }
   n10 = Component To Local(ComponentPose=@n9)
   n11 = Get bIsInFrontEnd()
   n12 = RigidBody(ComponentPose=Local To Component(LocalPose=Blend Poses by bool(BlendPose_0=@n6, BlendPose_1=@n10, BlendTime_0=0.2, BlendTime_1=0.2, bActiveValue=@n11) {blendType=Linear})) {
    bClampLinearTranslationLimitToRefPose=true
    bFreezeIncomingPoseOnStart=true
    bOverrideWorldGravity=true
    overridePhysicsAsset=&/F_MED_Uproar_Physics
    simulationSpace=BaseBoneSpace
    alphaBoolBlend {blendOption=Linear}
    baseBoneRef {boneName=spine_01}
    componentAppliedLinearAccClamp=<1000 1000 1000>
    componentLinearAccScale=<0.1 0.1 0.1>
    componentLinearVelScale=<0.1 0.1 0.1>
    overrideWorldGravity {z=-980}
   }
   n13 = RigidBody(ComponentPose=@n12) {
    bClampLinearTranslationLimitToRefPose=true
    bOverrideWorldGravity=true
    overridePhysicsAsset=&/F_MED_Uproar_Collar_Physics
    simulationSpace=BaseBoneSpace
    alphaBoolBlend {blendOption=Linear}
    alphaScaleBiasClamp {bInterpResult=true}
    baseBoneRef {boneName=spine_05}
    componentAppliedLinearAccClamp=<1000 1000 1000>
    componentLinearAccScale=<0.1 0.1 0.1>
    componentLinearVelScale=<0.1 0.1 0.1>
    overrideWorldGravity {z=-980}
   }
   n14 = Get LeftLeg_rotate() -> (LeftLeg_rotate:double)
   n15 = "Transform (Modify) Bone - Bone: Dyn_Belt1"(
    ComponentPose=Local To Component(LocalPose=Component To Local(ComponentPose=@n13))
    Alpha=@n14
    Translation="2,-1,-3"
    Rotation="0,0,0"
    Scale="1,1,1"
   ) {
    rotationMode=BMM_Additive
    rotationSpace=BCS_BoneSpace
    scaleMode=BMM_Additive
    scaleSpace=BCS_BoneSpace
    translationMode=BMM_Additive
    translationSpace=BCS_BoneSpace
    alphaBoolBlend {blendOption=Linear}
    boneToModify {boneName=Dyn_Belt1}
    translation=<2 -1 -3>
   }
   n16 = "Transform (Modify) Bone - Bone: Dyn_Master_Belt"(ComponentPose=@n15, Alpha=@n14, Translation="2,0,-1", Rotation="0,0,60", Scale="1,1,1") {
    rotationMode=BMM_Additive
    rotationSpace=BCS_BoneSpace
    scaleMode=BMM_Additive
    scaleSpace=BCS_BoneSpace
    translationMode=BMM_Additive
    translationSpace=BCS_BoneSpace
    alphaBoolBlend {blendOption=Linear}
    boneToModify {boneName=Dyn_Master_Belt}
    rotation {roll=60}
    translation {x=2, z=-1}
   }
   n17 = "Transform (Modify) Bone - Bone: Dyn_Master_Jacket_BL"(ComponentPose=@n16, Alpha=1, Translation="0,-1,0", Rotation="0,0,0", Scale="1,1,1") {
    rotationMode=BMM_Additive
    rotationSpace=BCS_BoneSpace
    scaleMode=BMM_Additive
    scaleSpace=BCS_BoneSpace
    translationMode=BMM_Additive
    translationSpace=BCS_BoneSpace
    alphaBoolBlend {blendOption=Linear}
    boneToModify {boneName=Dyn_Master_Jacket_BL}
    translation {y=-1}
   }
   n18 = "Transform (Modify) Bone - Bone: Dyn_Master_Jacket_BR"(ComponentPose=@n17, Alpha=1, Translation="0,-1,0", Rotation="0,0,0", Scale="1,1,1") {
    rotationMode=BMM_Additive
    rotationSpace=BCS_BoneSpace
    scaleMode=BMM_Additive
    scaleSpace=BCS_BoneSpace
    translationMode=BMM_Additive
    translationSpace=BCS_BoneSpace
    alphaBoolBlend {blendOption=Linear}
    boneToModify {boneName=Dyn_Master_Jacket_BR}
    translation {y=-1}
   }
   n19 = "Transform (Modify) Bone - Bone: Dyn_Master_Jacket_B"(ComponentPose=@n18, Alpha=1, Translation="0,-1,0", Rotation="0,0,0", Scale="1,1,1") {
    rotationMode=BMM_Additive
    rotationSpace=BCS_BoneSpace
    scaleMode=BMM_Additive
    scaleSpace=BCS_BoneSpace
    translationMode=BMM_Additive
    translationSpace=BCS_BoneSpace
    alphaBoolBlend {blendOption=Linear}
    boneToModify {boneName=Dyn_Master_Jacket_B}
    translation {y=-1}
   }
   n20 = Get bIsCrouching()
   n21 = "Transform (Modify) Bone - Bone: Dyn_Master_Jacket_FR"!(ComponentPose=@n19, bAlphaBoolEnabled:bool=@n20, Translation="-1,0,0", Rotation="25,0,0", Scale="1,1,1") -> (Pose) {
    alphaInputType=Bool
    rotationMode=BMM_Additive
    rotationSpace=BCS_BoneSpace
    scaleMode=BMM_Additive
    scaleSpace=BCS_BoneSpace
    translationMode=BMM_Additive
    translationSpace=BCS_BoneSpace
    alphaBoolBlend {blendInTime=0.1, blendOption=Linear, blendOutTime=0.1}
    boneToModify {boneName=Dyn_Master_Jacket_FR}
    rotation {pitch=25}
    translation {x=-1}
   }
   n22 = "Transform (Modify) Bone - Bone: Dyn_Master_Jacket_FL"!(ComponentPose=@n21, bAlphaBoolEnabled:bool=@n20, Translation="-2,0,0", Rotation="-10,0,0", Scale="1,1,1") -> (Pose) {
    alphaInputType=Bool
    rotationMode=BMM_Additive
    rotationSpace=BCS_BoneSpace
    scaleMode=BMM_Additive
    scaleSpace=BCS_BoneSpace
    translationMode=BMM_Additive
    translationSpace=BCS_BoneSpace
    alphaBoolBlend {blendInTime=0.1, blendOption=Linear, blendOutTime=0.1}
    boneToModify {boneName=Dyn_Master_Jacket_FL}
    rotation {pitch=-10}
    translation {x=-2}
   }
   n23 = "Transform (Modify) Bone - Bone: Dyn_Master_Collar__FR"(ComponentPose=@n22, Alpha=1, Translation="0,0,0", Rotation="0,-5,0", Scale="1,1,1") {
    rotationMode=BMM_Additive
    rotationSpace=BCS_BoneSpace
    scaleMode=BMM_Additive
    scaleSpace=BCS_BoneSpace
    translationMode=BMM_Additive
    translationSpace=BCS_BoneSpace
    alphaBoolBlend {blendOption=Linear}
    boneToModify {boneName=Dyn_Master_Collar__FR}
    rotation {yaw=-5}
   }
   n24 = "Transform (Modify) Bone - Bone: Dyn_Master_Collar__FR"(ComponentPose=@n23, Alpha=@n1, Translation="0,0,0", Rotation="0,-10,0", Scale="0.75,1,1") {
    rotationMode=BMM_Additive
    rotationSpace=BCS_BoneSpace
    scaleMode=BMM_Additive
    scaleSpace=BCS_BoneSpace
    translationMode=BMM_Additive
    translationSpace=BCS_BoneSpace
    alphaBoolBlend {blendOption=Linear}
    boneToModify {boneName=Dyn_Master_Collar__FR}
    rotation {yaw=-10}
    scale {x=0.75}
   }
   n25 = "Transform (Modify) Bone - Bone: Dyn_Master_Collar_R"(ComponentPose=@n24, Alpha=@n1, Translation="0,0,0", Rotation="0,5,0", Scale="1,1,1") {
    rotationMode=BMM_Additive
    rotationSpace=BCS_BoneSpace
    scaleMode=BMM_Additive
    scaleSpace=BCS_BoneSpace
    translationMode=BMM_Additive
    translationSpace=BCS_BoneSpace
    alphaBoolBlend {blendOption=Linear}
    boneToModify {boneName=Dyn_Master_Collar_R}
    rotation {yaw=5}
   }
   n26 = "Transform (Modify) Bone - Bone: Dyn_Master_Collar_FL"(ComponentPose=@n25, Alpha=@n2, Translation="0,0,0", Rotation="0,-5,0", Scale="0.6,1,1") {
    rotationMode=BMM_Additive
    rotationSpace=BCS_BoneSpace
    scaleMode=BMM_Additive
    scaleSpace=BCS_BoneSpace
    translationMode=BMM_Additive
    translationSpace=BCS_BoneSpace
    alphaBoolBlend {blendOption=Linear}
    boneToModify {boneName=Dyn_Master_Collar_FL}
    rotation {yaw=-5}
    scale {x=0.6}
   }
   n27 = "Transform (Modify) Bone - Bone: Dyn_Master_Collar_L"(ComponentPose=@n26, Alpha=@n2, Translation="0,0,0", Rotation="0,-1,0", Scale="0.6,0.6,0.6") {
    rotationMode=BMM_Additive
    rotationSpace=BCS_BoneSpace
    scaleMode=BMM_Additive
    scaleSpace=BCS_BoneSpace
    translationMode=BMM_Additive
    translationSpace=BCS_BoneSpace
    alphaBoolBlend {blendOption=Linear}
    boneToModify {boneName=Dyn_Master_Collar_L}
    rotation {yaw=-1}
    scale=<0.6 0.6 0.6>
   }
   n28 = Get bIsCrouching()
   n29 = "Transform (Modify) Bone - Bone: Dyn_Master_Jacket_FM_R2"!(ComponentPose=@n27, bAlphaBoolEnabled:bool=@n28, Translation="0,0,0", Rotation="0,0,0", Scale="0.7,1,1") -> (Pose) {
    alphaInputType=Bool
    rotationMode=BMM_Additive
    rotationSpace=BCS_BoneSpace
    scaleMode=BMM_Additive
    scaleSpace=BCS_BoneSpace
    translationMode=BMM_Additive
    translationSpace=BCS_BoneSpace
    alphaBoolBlend {blendInTime=0.1, blendOption=Linear, blendOutTime=0.1}
    boneToModify {boneName=Dyn_Master_Jacket_FM_R2}
    scale {x=0.7}
   }
   n30 = "Transform (Modify) Bone - Bone: Dyn_Master_Jacket_FR"(ComponentPose=@n29, bAlphaBoolEnabled=@n28, Translation="0,0,0", Rotation="0,0,0", Scale="0.7,1,1") {
    alphaInputType=Bool
    rotationMode=BMM_Additive
    rotationSpace=BCS_BoneSpace
    scaleMode=BMM_Additive
    scaleSpace=BCS_BoneSpace
    translationMode=BMM_Additive
    translationSpace=BCS_BoneSpace
    alphaBoolBlend {blendInTime=0.1, blendOption=Linear, blendOutTime=0.1}
    boneToModify {boneName=Dyn_Master_Jacket_FR}
    scale {x=0.7}
   }
   n31 = "Transform (Modify) Bone - Bone: Dyn_Master_Jacket_FM_L1"!(ComponentPose=@n30, bAlphaBoolEnabled:bool=@n28, Translation="0,0,0", Rotation="0,0,0", Scale="0.7,1,1") -> (Pose) {
    alphaInputType=Bool
    rotationMode=BMM_Additive
    rotationSpace=BCS_BoneSpace
    scaleMode=BMM_Additive
    scaleSpace=BCS_BoneSpace
    translationMode=BMM_Additive
    translationSpace=BCS_BoneSpace
    alphaBoolBlend {blendInTime=0.1, blendOption=Linear, blendOutTime=0.1}
    boneToModify {boneName=Dyn_Master_Jacket_FM_L1}
    scale {x=0.7}
   }
   n32 = "Transform (Modify) Bone - Bone: Dyn_Master_Jacket_FL"(ComponentPose=@n31, bAlphaBoolEnabled=@n28, Translation="0,0,0", Rotation="0,0,0", Scale="0.7,1,1") {
    alphaInputType=Bool
    rotationMode=BMM_Additive
    rotationSpace=BCS_BoneSpace
    scaleMode=BMM_Additive
    scaleSpace=BCS_BoneSpace
    translationMode=BMM_Additive
    translationSpace=BCS_BoneSpace
    alphaBoolBlend {blendInTime=0.1, blendOption=Linear, blendOutTime=0.1}
    boneToModify {boneName=Dyn_Master_Jacket_FL}
    scale {x=0.7}
   }
   Output Pose(Result=Component To Local(ComponentPose=@n32)) {layerGroup=None}
  ]}
 EventGraph {nodes [
   Event BlueprintUpdateAnimation()
   > n1 = Sequence!(execute) -> (then_0, then_1, then_2:Exec, then_3:Exec)
   > Set Right Arm ADS Up(Right Arm ADS Up:double=GetAlphaFromReferenceTransformDelta(SkeletalMeshComponent=Get Skel Mesh(), BoneName=upperarm_r, BaseName=clavicle_r, InputRange=(X=35,Y=112), TwistAxis="0,0,1")) -> (Output_Get:double)
   "Set Left  Arm ADS Up"(
    execute=@n1.then_1
    "Left  Arm ADS Up":double=GetAlphaFromReferenceTransformDelta(SkeletalMeshComponent=Get Skel Mesh(), BoneName=upperarm_l, BaseName=clavicle_l, InputRange=(X=46,Y=68), TwistAxis="0,0,1")
   ) -> (Output_Get:double)
   Set Left_Hood(
    execute=@n1.then_2
    Left_Hood:double=GetAlphaFromReferenceTransformDelta(SkeletalMeshComponent=Get Skel Mesh(), BoneName=spine_05, BaseName=clavicle_l, InputRange=(X=-2,Y=36), TwistAxis="0,0,1")
   ) -> (Output_Get:double)
   Set LeftLeg_rotate(
    execute=@n1.then_3
    LeftLeg_rotate:double=GetAlphaFromReferenceTransformDelta(SkeletalMeshComponent=Get Skel Mesh(), BoneName=thigh_l, BaseName=pelvis, InputRange=(X=20,Y=-104), TwistAxis="0,0,1")
   ) -> (Output_Get:double)
   Event BlueprintInitializeAnimation()
   > Set Skel Mesh(Skel Mesh=GetOwnerSkeletalMeshComponent())
  ]}
]
variables [
 "Left  Arm ADS Up" {type=number}
 Left_Hood {type=number}
 LeftLeg_rotate {type=number}
 Right Arm ADS Up {type=number}
 Skel Mesh {type=SkeletalMeshComponent}
]
