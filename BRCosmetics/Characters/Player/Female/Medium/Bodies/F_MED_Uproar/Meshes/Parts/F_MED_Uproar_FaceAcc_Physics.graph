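# Physics asset for the F_MED_Uproar face accessory mesh, defining capsule bodies and joint limits for the character's dynamic hair bones.
# Short hair pieces each constrain directly to the head body, which carries two capsules.
# Two front locks and two right-side strands form multi-bone chains rooted on small Master bodies, which have no constraints to the head.
# Joints use narrow swing limits and minimal twist to restrict hair motion.
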PhysicsAsset /BRCosmetics/Characters/Player/Female/Medium/Bodies/F_MED_Uproar/Meshes/Parts/F_MED_Uproar_FaceAcc_Physics
bodies {
 DynHair_10 [{shapeType=Capsule, length=2.1046, radius=0.5906, shapeName=DynHair_10_capsule, center=<0.3 0 0>, rotation=<rot 90 0 0>}]
 Dyn_Front2_Hair1 [{shapeType=Capsule, length=4, radius=0.5, shapeName=Dyn_Front2_Hair1_capsule, center=<2 0 0>, rotation=<rot 90 0 0>}]
 Dyn_Front2_Hair2 [{
   shapeType=Capsule
   length=5
   radius=0.4
   shapeName=Dyn_Front2_Hair2_capsule
   center=<1.8805 0.3345 -0.1636>
   rotation=<rot 79.9991 15.0014 0>
  }]
 Dyn_Front_Hair_01 [{shapeType=Capsule, length=4, radius=0.4, shapeName=Dyn_Front_Hair_01_capsule, center=<2 -0.3 0>, rotation=<rot 90 0 0>}]
 Dyn_Front_Hair_02 [{shapeType=Capsule, length=5, radius=0.4, shapeName=Dyn_Front_Hair_02_capsule, center=<2 0 -0.3>, rotation=<rot 90 0 0>}]
 Dyn_Hair1 [{shapeType=Capsule, length=1.3339, radius=0.6401, shapeName=Dyn_Hair1_capsule, center=<1 0 0>, rotation=<rot 90 0 0>}]
 Dyn_Hair2 [{shapeType=Capsule, length=4.5144, radius=1.2387, shapeName=Dyn_Hair2_capsule, center=<1 0 0>, rotation=<rot 90 0 0>}]
 Dyn_Hair3 [{shapeType=Capsule, length=3.2444, radius=1.2565, shapeName=Dyn_Hair3_capsule, center=<1 0 0>, rotation=<rot 90 0 0>}]
 Dyn_Hair4 [{shapeType=Capsule, length=1.8895, radius=0.7159, shapeName=Dyn_Hair4_capsule, center=<1 0 0>, rotation=<rot 90 0 0>}]
 Dyn_Hair5 [{shapeType=Capsule, length=4.1753, radius=2.1155, shapeName=Dyn_Hair5_capsule, center=<1 0 0>, rotation=<rot 90 0 0>}]
 Dyn_Hair6 [{shapeType=Capsule, length=2.8527, radius=1.0165, shapeName=Dyn_Hair6_capsule, center=<1 0 0>, rotation=<rot 90 0 0>}]
 Dyn_Hair7 [{shapeType=Capsule, length=1, radius=0.5, shapeName=Dyn_Hair7_capsule, center=<0.5 0 0>, rotation=<rot 90 0 0>}]
 Dyn_Hair8 [{shapeType=Capsule, length=2.1254, radius=0.5, shapeName=Dyn_Hair8_capsule, center=<0.5 0 0>, rotation=<rot 90 0 0>}]
 Dyn_Hair9 [{shapeType=Capsule, length=1.6466, radius=0.5705, shapeName=Dyn_Hair9_capsule, center=<0.5 0 0>, rotation=<rot 90 0 0>}]
 Dyn_Hair_11 [{shapeType=Capsule, length=1.6941, radius=0.8103, shapeName=Dyn_Hair_11_capsule, center=<0.7 0 0>, rotation=<rot 90 0 0>}]
 Dyn_Hair_12 [{shapeType=Capsule, length=1.7737, radius=0.9156, shapeName=Dyn_Hair_12_capsule, center=<0.7 0 0>, rotation=<rot 90 0 0>}]
 Dyn_Hair_13 [{shapeType=Capsule, length=1.8672, radius=0.8736, shapeName=Dyn_Hair_13_capsule, center=<0.7 0 0>, rotation=<rot 90 0 0>}]
 Dyn_Hair_14 [{shapeType=Capsule, length=2.229, radius=0.88, shapeName=Dyn_Hair_14_capsule, center=<0.8 0 0>, rotation=<rot 90 0 0>}]
 Dyn_Hair_15 [{shapeType=Capsule, length=1.4971, radius=0.7099, shapeName=Dyn_Hair_15_capsule, center=<0.8 0 0>, rotation=<rot 90 0 0>}]
 Dyn_Hair_16 [{shapeType=Capsule, length=2.1068, radius=0.6677, shapeName=Dyn_Hair_16_capsule, center=<0.2 0 0>, rotation=<rot 90 0 0>}]
 Dyn_Hair_17 [{shapeType=Capsule, length=2.1139, radius=0.832, shapeName=Dyn_Hair_17_capsule, center=<0.7 0 0>, rotation=<rot 90 0 0>}]
 Dyn_Hair_18 [{shapeType=Capsule, length=1.6602, radius=0.6918, shapeName=Dyn_Hair_18_capsule, center=<0.7 0 0>, rotation=<rot 90 0 0>}]
 Dyn_Hair_24 [{shapeType=Capsule, length=0.505, radius=0.505, shapeName=Dyn_Hair_24_capsule, center=<1 0 0>, rotation=<rot 90 0 0>}]
 Dyn_Master_Front2_Hair [{shapeType=Capsule, length=0.5, radius=0.5, shapeName=Dyn_Master_Front2_Hair_capsule}]
 Dyn_Master_Front_Hair [{shapeType=Capsule, length=0.5, radius=0.5, shapeName=Dyn_Master_Front_Hair_capsule}]
 Dyn_Master_Right_Hairstrand [{shapeType=Capsule, length=0.5, radius=0.5, shapeName=Dyn_Master_Right_Hairstrand_capsule}]
 Dyn_Master_Right_Hairstrand_Inner1 [{shapeType=Capsule, length=0.5, radius=0.5, shapeName=Dyn_Master_Right_Hairstrand_Inner1_capsule}]
 Dyn_Right_HairStrand_02 [{shapeType=Capsule, length=4, radius=0.75, shapeName=Dyn_Right_HairStrand_02_capsule, center=<2 0 0>, rotation=<rot 90 0 0>}]
 Dyn_Right_HairStrand_03 [{shapeType=Capsule, length=4, radius=0.75, shapeName=Dyn_Right_HairStrand_03_capsule, center=<2 0 0>, rotation=<rot 90 0 0>}]
 Dyn_Right_Hairstrand_01 [{shapeType=Capsule, length=4, radius=0.75, shapeName=Dyn_Right_Hairstrand_01_capsule, center=<0.5 0 0>, rotation=<rot 90 0 0>}]
 Dyn_Right_Hairstrand_Inner1 [{shapeType=Capsule, length=4, radius=0.6, shapeName=Dyn_Right_Hairstrand_Inner1_capsule, center=<2 -0.35 0>, rotation=<rot 90 0 0>}]
 Dyn_Right_Hairstrand_Inner2 [{shapeType=Capsule, length=6, radius=0.6, shapeName=Dyn_Right_Hairstrand_Inner2_capsule, center=<3.2365 -0.1448 0>, rotation=<rot 90 0 0>}]
 head [
  {shapeType=Capsule, length=6, radius=6, shapeName=head_capsule, center=<3.2931 3.5247 0>, rotation=<rot 90 0 0>}
  {shapeType=Capsule, length=3, radius=1, shapeName=head_capsule, center=<1.2584 8.9676 0>, rotation=<rot 90 90 105.0013>}
 ]
}
constraints [
 {bone1Name=Dyn_Right_Hairstrand_01, bone2Name=Dyn_Master_Right_Hairstrand, swing1LimitDegrees=20, swing2LimitDegrees=20, twistLimitDegrees=1}
 {bone1Name=Dyn_Right_HairStrand_02, bone2Name=Dyn_Right_Hairstrand_01, swing1LimitDegrees=10, swing2LimitDegrees=30, twistLimitDegrees=1}
 {bone1Name=Dyn_Right_HairStrand_03, bone2Name=Dyn_Right_HairStrand_02, swing1LimitDegrees=20, swing2LimitDegrees=20, twistLimitDegrees=1}
 {bone1Name=Dyn_Front2_Hair1, bone2Name=Dyn_Master_Front2_Hair, swing1LimitDegrees=20, swing2LimitDegrees=20, twistLimitDegrees=1}
 {bone1Name=Dyn_Front2_Hair2, bone2Name=Dyn_Front2_Hair1, swing1LimitDegrees=30, swing2LimitDegrees=30, twistLimitDegrees=1}
 {bone1Name=Dyn_Front_Hair_01, bone2Name=Dyn_Master_Front_Hair, swing1LimitDegrees=20, swing2LimitDegrees=20, twistLimitDegrees=1}
 {bone1Name=Dyn_Front_Hair_02, bone2Name=Dyn_Front_Hair_01, swing1LimitDegrees=30, swing2LimitDegrees=30, twistLimitDegrees=1}
 {
  bone1Name=Dyn_Right_Hairstrand_Inner1
  bone2Name=Dyn_Master_Right_Hairstrand_Inner1
  swing1LimitDegrees=10
  swing2LimitDegrees=20
  twistLimitDegrees=1
 }
 {
  bone1Name=Dyn_Right_Hairstrand_Inner2
  bone2Name=Dyn_Right_Hairstrand_Inner1
  swing1LimitDegrees=10
  swing2LimitDegrees=20
  twistLimitDegrees=1
 }
 {bone1Name=Dyn_Hair_24, bone2Name=head, swing1LimitDegrees=10, swing2LimitDegrees=10, twistLimitDegrees=1}
 {bone1Name=Dyn_Hair1, bone2Name=head, swing1LimitDegrees=10, swing2LimitDegrees=10, twistLimitDegrees=1}
 {bone1Name=Dyn_Hair2, bone2Name=head, swing1LimitDegrees=20, swing2LimitDegrees=10, twistLimitDegrees=1}
 {bone1Name=Dyn_Hair3, bone2Name=head, swing1LimitDegrees=10, swing2LimitDegrees=10, twistLimitDegrees=1}
 {bone1Name=Dyn_Hair4, bone2Name=head, swing1LimitDegrees=10, swing2LimitDegrees=10, twistLimitDegrees=1}
 {bone1Name=Dyn_Hair5, bone2Name=head, swing1LimitDegrees=10, swing2LimitDegrees=30, twistLimitDegrees=1}
 {bone1Name=Dyn_Hair6, bone2Name=head, swing1LimitDegrees=10, swing2LimitDegrees=10, twistLimitDegrees=1}
 {bone1Name=Dyn_Hair7, bone2Name=head, swing1LimitDegrees=15, swing2LimitDegrees=5, twistLimitDegrees=1}
 {bone1Name=Dyn_Hair8, bone2Name=head, swing1LimitDegrees=20, swing2LimitDegrees=2, twistLimitDegrees=1}
 {bone1Name=Dyn_Hair9, bone2Name=head, swing1LimitDegrees=20, swing2LimitDegrees=1, twistLimitDegrees=1}
 {bone1Name=DynHair_10, bone2Name=head, swing1LimitDegrees=10, swing2LimitDegrees=10, twistLimitDegrees=1}
 {bone1Name=Dyn_Hair_11, bone2Name=head, swing1LimitDegrees=15, swing2LimitDegrees=10, twistLimitDegrees=1}
 {bone1Name=Dyn_Hair_12, bone2Name=head, swing1LimitDegrees=15, swing2LimitDegrees=10, twistLimitDegrees=1}
 {bone1Name=Dyn_Hair_13, bone2Name=head, swing1LimitDegrees=15, swing2LimitDegrees=10, twistLimitDegrees=1}
 {bone1Name=Dyn_Hair_14, bone2Name=head, swing1LimitDegrees=20, swing2LimitDegrees=10, twistLimitDegrees=1}
 {bone1Name=Dyn_Hair_15, bone2Name=head, swing1LimitDegrees=10, swing2LimitDegrees=10, twistLimitDegrees=1}
 {bone1Name=Dyn_Hair_16, bone2Name=head, swing1LimitDegrees=10, swing2LimitDegrees=10, twistLimitDegrees=1}
 {bone1Name=Dyn_Hair_17, bone2Name=head, swing1LimitDegrees=10, swing2LimitDegrees=10, twistLimitDegrees=1}
 {bone1Name=Dyn_Hair_18, bone2Name=head, swing1LimitDegrees=10, swing2LimitDegrees=10, twistLimitDegrees=1}
]
